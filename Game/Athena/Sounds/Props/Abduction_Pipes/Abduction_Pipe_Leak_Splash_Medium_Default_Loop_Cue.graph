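# Looping ambient sound cue for a medium water splash from a leaking Abduction pipe, on the outdoor ambient sound class.
# A quality switch gives the first level three random loop variants and the other three levels a shared two-variant branch with a narrower pitch range.
# Each branch randomizes pitch and start delay around its chosen loop.
# The cue uses a dedicated attenuation asset and caps simultaneous instances through a concurrency override.
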SoundCue /Game/Athena/Sounds/Props/Abduction_Pipes/Abduction_Pipe_Leak_Splash_Medium_Default_Loop_Cue
root=QualityLevel_0
nodes [
 {id=Delay_0, children [Random_0], properties {delayMax=0.5}}
 {id=Delay_3, children [Random_1], properties {delayMax=0.5}}
 {id=Modulator_1, children [Delay_0], properties {pitchMax=1.1, pitchMin=0.8, volumeMax=1, volumeMin=1}}
 {id=Modulator_2, children [Delay_3], properties {pitchMax=1.1, pitchMin=0.9, volumeMax=1, volumeMin=1}}
 {id=QualityLevel_0, children [Modulator_1,Modulator_2,Modulator_2,Modulator_2]}
 {id=Random_0, children [WavePlayer_0,WavePlayer_7,WavePlayer_8], properties {bShouldExcludeFromBranchCulling=true, weights [1,1,1]}}
 {id=Random_1, children [WavePlayer_10, WavePlayer_11], properties {bShouldExcludeFromBranchCulling=true, weights [1,1]}}
 {
  id=WavePlayer_0
  properties {bLooping=true, soundWaveAssetPtr=&/Abduction_Pipe_Leak_Water_Impact_Medium_Default_Loop_01}
 }
 {
  id=WavePlayer_10
  properties {bLooping=true, soundWaveAssetPtr=&Abduction_Pipe_Leak_Water_Impact_Medium_Default_Loop_01}
 }
 {
  id=WavePlayer_11
  properties {bLooping=true, soundWaveAssetPtr=&/Abduction_Pipe_Leak_Water_Impact_Medium_Default_Loop_02}
 }
 {
  id=WavePlayer_7
  properties {bLooping=true, soundWaveAssetPtr=&Abduction_Pipe_Leak_Water_Impact_Medium_Default_Loop_02}
 }
 {
  id=WavePlayer_8
  properties {bLooping=true, soundWaveAssetPtr=&/Abduction_Pipe_Leak_Water_Impact_Medium_Default_Loop_03}
 }
]
properties {
 attenuationSettings=&/Abduction_Pipe_Leak_Splash_Impact_Loop_Att
 bOverrideConcurrency=true
 maxDistance=1573.7499
 priority=10
 soundClassObject=&Fort_Ambient_Outdoors_Athena
 volumeMultiplier=0.301
 concurrencyOverrides {maxCount=12}
}
