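# Looping sound cue for a small water splash from leaking Abduction pipes, an outdoor ambient prop sound that reuses Medium water-impact loops.
# A quality-level switch picks between a three-loop random branch in the first slot and a two-loop branch shared by the other three.
# Each branch adds a random start delay and slight pitch variation.
# Attenuation is shared through an asset, and a concurrency override caps instances, stopping the quietest.
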SoundCue /Game/Athena/Sounds/Props/Abduction_Pipes/Abduction_Pipe_Leak_Splash_Small_Default_Loop_Cue
root=QualityLevel_1
nodes [
 {id=Delay_0, children [Random_2], properties {delayMax=1.5}}
 {id=Delay_3, children [Random_3], properties {delayMax=1.5}}
 {id=Modulator_0, children [Delay_0], properties {pitchMax=1.1, pitchMin=0.9, volumeMax=1, volumeMin=1}}
 {id=Modulator_3, children [Delay_3], properties {pitchMax=1.1, pitchMin=0.9, volumeMax=1, volumeMin=1}}
 {id=QualityLevel_1, children [Modulator_0,Modulator_3,Modulator_3,Modulator_3]}
 {id=Random_2, children [WavePlayer_2,WavePlayer_3,WavePlayer_4], properties {bShouldExcludeFromBranchCulling=true, weights [1,1,1]}}
 {id=Random_3, children [WavePlayer_5,WavePlayer_6], properties {bShouldExcludeFromBranchCulling=true, weights [1,1]}}
 {
  id=WavePlayer_2
  properties {bLooping=true, soundWaveAssetPtr=&/Abduction_Pipe_Leak_Water_Impact_Medium_Default_Loop_01}
 }
 {
  id=WavePlayer_3
  properties {bLooping=true, soundWaveAssetPtr=&/Abduction_Pipe_Leak_Water_Impact_Medium_Default_Loop_02}
 }
 {
  id=WavePlayer_4
  properties {bLooping=true, soundWaveAssetPtr=&/Abduction_Pipe_Leak_Water_Impact_Medium_Default_Loop_03}
 }
 {
  id=WavePlayer_5
  properties {bLooping=true, soundWaveAssetPtr=&Abduction_Pipe_Leak_Water_Impact_Medium_Default_Loop_01}
 }
 {
  id=WavePlayer_6
  properties {bLooping=true, soundWaveAssetPtr=&Abduction_Pipe_Leak_Water_Impact_Medium_Default_Loop_02}
 }
]
properties {
 attenuationSettings=&/Abduction_Pipe_Leak_Splash_Impact_Loop_Att
 bOverrideConcurrency=true
 maxDistance=1573.7499
 priority=10
 soundClassObject=&Fort_Ambient_Outdoors_Athena
 volumeMultiplier=0.301
 concurrencyOverrides {maxCount=2, resolutionRule=StopQuietest}
}
