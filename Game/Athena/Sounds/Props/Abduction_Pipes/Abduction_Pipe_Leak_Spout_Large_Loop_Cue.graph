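# Looping outdoor ambient sound cue for a large spouting leak on Athena's Abduction pipe props, attenuated by distance with its own concurrency limit.
# A quality-level node routes its first two slots to a random pick of four looping waves, and its last two to a pick of two separately referenced waves.
# Both branches add slight pitch variation, a random start delay and a curved volume fade-in, staggering and softening each emitter's start.
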SoundCue /Game/Athena/Sounds/Props/Abduction_Pipes/Abduction_Pipe_Leak_Spout_Large_Loop_Cue
root=QualityLevel_0
nodes [
 {id=Delay_1, children [Enveloper_1], properties {delayMax=1.5}}
 {id=Delay_2, children [Enveloper_0], properties {delayMax=1.5}}
 {
  id=Enveloper_0
  children [Random_1]
  properties {
   pitchCurve {editorCurveData {keys [{time=0}]}}
   volumeCurve {editorCurveData {keys [<at -0.0001 0 RCIM_Cubic>, {arriveTangent=0.4839, interpMode=RCIM_Cubic, leaveTangent=0.4839, tangentMode=RCTM_User, time=3, value=1}]}}
  }
 }
 {
  id=Enveloper_1
  children [Random_0]
  properties {
   pitchCurve {editorCurveData {keys [{time=0}]}}
   volumeCurve {editorCurveData {keys [<at -0.0001 0 RCIM_Cubic>, {arriveTangent=0.4839, interpMode=RCIM_Cubic, leaveTangent=0.4839, tangentMode=RCTM_User, time=3, value=1}]}}
  }
 }
 {id=Modulator_1, children [Delay_1], properties {pitchMax=1.1, pitchMin=0.9, volumeMax=1, volumeMin=1}}
 {id=Modulator_2, children [Delay_2], properties {pitchMax=1.1, pitchMin=0.9, volumeMax=1, volumeMin=1}}
 {id=QualityLevel_0, children [Modulator_1,Modulator_1,Modulator_2,Modulator_2]}
 {
  id=Random_0
  children [WavePlayer_1,WavePlayer_2,WavePlayer_3,WavePlayer_4]
  properties {bShouldExcludeFromBranchCulling=true, weights [1,1,1,1]}
 }
 {id=Random_1, children [WavePlayer_5,WavePlayer_6], properties {bShouldExcludeFromBranchCulling=true, weights [1,1]}}
 {
  id=WavePlayer_1
  properties {bLooping=true, soundWaveAssetPtr=&/Abduction_Pipe_Leak_Spout_Large_Loop_01}
 }
 {
  id=WavePlayer_2
  properties {bLooping=true, soundWaveAssetPtr=&/Abduction_Pipe_Leak_Spout_Large_Loop_02}
 }
 {
  id=WavePlayer_3
  properties {bLooping=true, soundWaveAssetPtr=&/Abduction_Pipe_Leak_Spout_Large_Loop_03}
 }
 {
  id=WavePlayer_4
  properties {bLooping=true, soundWaveAssetPtr=&/Abduction_Pipe_Leak_Spout_Large_Loop_04}
 }
 {
  id=WavePlayer_5
  properties {bLooping=true, soundWaveAssetPtr=&Abduction_Pipe_Leak_Spout_Large_Loop_01}
 }
 {
  id=WavePlayer_6
  properties {bLooping=true, soundWaveAssetPtr=&Abduction_Pipe_Leak_Spout_Large_Loop_02}
 }
]
properties {
 attenuationSettings=&/Abduction_Pipe_Leak_Spout_Loop_Att
 bOverrideConcurrency=true
 maxDistance=1623.7499
 priority=10
 soundClassObject=&Fort_Ambient_Outdoors_Athena
 volumeMultiplier=0.351
 concurrencyOverrides {maxCount=10}
}
