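# Sound cue for the looping ambient sound of an abduction pipe prop in Athena, on the outdoor ambient sound class.
# A looping wave runs through a Modulator for random pitch variation, then an Enveloper whose volume curve fades it in from silence.
# Dedicated attenuation settings shape it over distance, and a concurrency override caps it at two instances, stopping the quietest.
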SoundCue /Game/Athena/Sounds/Props/Abduction_Pipes/Abduction_Pipe_Loop_Cue
root=Enveloper_1
nodes [
 {
  id=Enveloper_1
  children [Modulator_0]
  properties {pitchCurve {editorCurveData {keys [{time=0}]}}, volumeCurve {editorCurveData {keys [<at -0.0001 0>,<at 1 1>]}}}
 }
 {id=Modulator_0, children [WavePlayer_0], properties {pitchMax=1.2}}
 {id=WavePlayer_0, properties {bLooping=true, soundWaveAssetPtr=&/Abduction_Pipe_Loop}}
]
properties {
 attenuationSettings=&/Abduction_Pipe_Loop_Att
 bOverrideConcurrency=true
 maxDistance=3047.4998
 priority=10
 soundClassObject=&Fort_Ambient_Outdoors_Athena
 volumeMultiplier=1.002
 concurrencyOverrides {maxCount=2, resolutionRule=StopQuietest}
}
